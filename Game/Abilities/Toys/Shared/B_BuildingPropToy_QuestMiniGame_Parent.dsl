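; Parent building prop toy for quest mini-games that credits a scoring player's quest objective and plays success feedback.
; On playlist load it flags Creative mode, then remote instances wait before checking the local player's quest.
; An incomplete quest outside Creative binds to quest updates, which drive attached FX, sound and force feedback before unbinding.
; PlayerScored plays score force feedback and sends a quest update if the objective is incomplete.
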(event EventBeginPlay
  (|CallParentFunction)
  ;[1 1 1] Check if In Creative Mode or not
  (bind _asfort_game_state_athena (Utilities|Casting|CastToFortGameStateAthena (Game|GetGameState)))
  (if (Variables|/Script/FortniteGame.FortGameStateAthena|GetbPlaylistDataIsLoaded _asfort_game_state_athena)
    (if (GameplayTags|HasTag (Class|FortGameStateAthena|GetAthenaPlaylistContextTags _asfort_game_state_athena) "(TagName=\"Athena.Creative\")")
      (Variables|QuestStuff|SetisCreativeMode true)
      (else
        (Variables|QuestStuff|SetisCreativeMode false)))
    (Networking|SwitchHasAuthority
      (:Remote
        (Utilities|FlowControl|Delay 5.0)
        ;[1 1 1] Check for Quest Manager to be Loaded
        (CallFunction|QuestManagerandPlaylistLoaded)))
    (else
      (Playlist|BindEventtoOnCurrentPlaylistLoaded _asfort_game_state_athena (delegate Custom|PlaylistLoaded)))))

(event Custom|QuestManagerLoaded
  ;[1 1 1] Check for Quest Manager to be Loaded
  (CallFunction|QuestManagerandPlaylistLoaded))

(event Custom|PlaylistLoaded
  (params (PlaylistName :type Name) (PlaylistContextTags :type Gameplay%20Tag%20Container%20Structure%20%28by%20ref%29))
  ;[1 1 1] Check if In Creative Mode or not
  (if (GameplayTags|HasTag PlaylistContextTags "(TagName=\"Athena.Creative\")")
    (Variables|QuestStuff|SetisCreativeMode true)
    (else
      (Variables|QuestStuff|SetisCreativeMode false)))
  (Networking|SwitchHasAuthority
    (:Remote
      (Utilities|FlowControl|Delay 5.0)
      ;[1 1 1] Check for Quest Manager to be Loaded
      (CallFunction|QuestManagerandPlaylistLoaded))))

(event Custom|QuestManagerAndPlaylistLoaded
  ;[1 1 1] PlayerHasQuest and is not in Creative Mode
  (bind (_questisvalid _questcompleted) (Default|AthenaCheckQuestandObjectiveCompletion (Utilities|Array|Get(acopy) (Fortnite|GetLocalFortPlayerControllers true)) (Variables|QuestStuff|GetQuestItem) (Variables|QuestStuff|GetObjBackendName)))
  (if (and _questisvalid (not (Variables|QuestStuff|GetIsCreativeMode)))
    (if (not _questcompleted)
      (Variables|QuestStuff|SetQuestLocallyActive true)
      ;[1 1 1] Bind and unbind Quest Updated Events
      (Quests|BindEventtoOnQuestsUpdated (Quests|GetQuestManager (Utilities|Array|Get(acopy) (Fortnite|GetLocalFortPlayerControllers true)) "Invalid") (EventDispatchers|CreateEvent :!event Custom|QuestUpdated)))))

(event Custom|QuestUpdated
  (bind _returnvalue (Fortnite|GetLocalFortPlayerControllers true))
  (bind _output (Utilities|Array|Get(acopy) _returnvalue))
  (bind _returnvalue_2 (Game|Player|GetPlayerPawn _output))
  ;[1 1 1] Check if Valid for SuccessFX
  (if (Variables|QuestStuff|GetQuestLocallyActive)
    (bind (_questisvalid _questcompleted _objectivecompleted) (Default|AthenaCheckQuestandObjectiveCompletion (Utilities|Array|Get(acopy) (Fortnite|GetLocalFortPlayerControllers true)) (Variables|QuestStuff|GetQuestItem) (Variables|QuestStuff|GetObjBackendName)))
    (if _questcompleted
      _
      (elif _objectivecompleted
        _))
    (Utilities|FlowControl|Delay 0.5)
    (Effects|SpawnFXSystemAttached (Variables|QuestStuff|GetQuestSuccessFX) (Variables|/Script/Engine.Character|GetMesh _returnvalue_2))
    (Audio|PlaySoundatLocation (Variables|QuestStuff|GetQuestSuccessSound) (Transformation|GetActorLocation _returnvalue_2))
    (Game|Feedback|ClientPlayForceFeedback (Utilities|Array|Get(acopy) (Fortnite|GetLocalFortPlayerControllers true)) (Variables|QuestStuff|GetQuestForceFeedback) "None" false false false)
    ;[1 1 1] Bind and unbind Quest Updated Events
    (Variables|QuestStuff|SetQuestLocallyActive false)
    (Quests|UnbindEventfromOnQuestsUpdated (Quests|GetQuestManager (Utilities|Array|Get(acopy) (Fortnite|GetLocalFortPlayerControllers true)) "Invalid") (EventDispatchers|CreateEvent :!event Custom|QuestUpdated))))

(fn PlayerScored
  (params (FortPlayerController :type Fort%20Player%20Controller%20Object%20Reference))
  (bind _questitem (Variables|QuestStuff|GetQuestItem))
  (bind _objbackendname (Variables|QuestStuff|GetObjBackendName))
  (Game|Feedback|ClientPlayForceFeedback FortPlayerController (Variables|Default|GetScoreForceFeedback) "None" false false false)
  (bind (_questisvalid _questcompleted _objectivecompleted) (Default|AthenaCheckQuestandObjectiveCompletion FortPlayerController _questitem _objbackendname))
  (if (not _objectivecompleted)
    (Class|AthenaQuestFunctionLibrary|AthenaGiveQuestUpdatetoPlayers :InQuestItemRef _questitem :ObjectiveBackendName _objbackendname :PlayerControllersForUpdate (Utilities|Array|MakeArray FortPlayerController))))

(fn ConstructionScript
  (|CallParentFunction))
