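; Player-spawned ball toy that handles its placement, reuse, expiry and frontend preview.
; Initialization records the owning pawn and spawn position and zeroes projectile velocity; BeginPlay flags it as attached to the player.
; Reuse and fade-out play optional replicated gameplay cues at the ball's location with the owner as instigator, and fade-out then calls SilentDie.
; Frontend preview deactivates projectile movement, forces a mesh LOD and sets a preview flag.
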(event EventBeginPlay
  (CallFunction|SetIsAttachedtoPlayer true)
  (|CallParentFunction))

(event Visuals|EventNotifyToyInstanceOfReuse
  (bind _cuetoplaywhenreused (Variables|Default|GetCueToPlayWhenReused))
  (if (GameplayTags|IsGameplayTagValid _cuetoplaywhenreused)
    (CallFunction|PlayCueCenteredonMe _cuetoplaywhenreused)))

(event Visuals|EventStartToyFadeOutDueToNewPlacement
  (bind _cuetoplaywhenexpired (Variables|Default|GetCueToPlayWhenExpired))
  (if (GameplayTags|IsGameplayTagValid _cuetoplaywhenexpired)
    (CallFunction|PlayCueCenteredonMe _cuetoplaywhenexpired)
    (else
      _))
  (Gameplay|SilentDie false))

(event Visuals|EventInitializeToyInstance
  (params (OwningPC :type Fort%20Player%20Controller%20Object%20Reference) (NumTimesSummoned :type Integer))
  (bind _outpawn (Class|AbilityFunctions|GetPawnfromPlayerStateZone (Variables|/Script/Engine.Controller|GetPlayerState OwningPC)))
  (Variables|Default|SetOwningPawn _outpawn)
  (Variables|Default|SetInitialSpawnPosition (Transformation|GetActorLocation))
  (Variables|/Script/Engine.MovementComponent|SetVelocity "0, 0, 0" (Variables|Default|GetProjectileMovement)))

(event Visuals|EventInitializeToyPreviewInFrontend
  (Components|Activation|SetActive (Variables|Default|GetProjectileMovement) false false)
  (Rendering|LOD|SetForcedLodModel (Variables|Default|GetBall) 1)
  (Variables|Default|SetbIsFrontEndPreview true))

(fn PlayCueCenteredOnMe
  (params (GameplayCueToPlay :type Gameplay%20Tag%20Structure) (Raw%20Magnitude :type Float%20%28double-precision%29))
  (Ability|ExecuteGameplayCueWithParamsReplicated GameplayCueToPlay (Ability|GameplayCue|MakeGameplayCueParameters :NormalizedMagnitude 0.0 :RawMagnitude Raw%20Magnitude :EffectContext "()" :AggregatedSourceTags "(GameplayTags=)" :AggregatedTargetTags "(GameplayTags=)" :Location (Math|Vector|MakeVectorNetQuantize10 (.x (Transformation|GetActorLocation)) (.y (Transformation|GetActorLocation)) (.z (Transformation|GetActorLocation))) :Normal "0,0,0" :Instigator (Variables|Default|GetOwningPawn) :GameplayEffectLevel 1 :AbilityLevel 1 :bReplicateLocationWhenUsingMinimalRepProxy false)))

(fn ConstructionScript
  (|CallParentFunction))
